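# Translucent unlit particle sprite material, named as the dot for a design utility that highlights actors, with depth test disabled so it draws through geometry.
# UV distance from the sprite center, measured against a radius parameter, feeds a SmoothStep with Min above Max, giving a soft procedural ring near the edge without textures.
# Particle color RGB passes through MF_FX_EmissiveAdjustments for the emissive value.
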Material /Game/Athena/Prototype/DesignUtilities/M_HighlightActor_Blutility_Dot
graph {
 properties {
  bDisableDepthTest=true
  bUsedWithParticleSprites=true
  blendMode=BLEND_Translucent
  pixelDepthOffsetMode=PDOM_Legacy
  shadingModel=MSM_Unlit
  usedShadingModels=Unlit
 }
 nodes [
  n1 = Distance(A=TextureCoordinate(), B=Constant() {r=0.5})
  MaterialFunctionCall(
   Alpha=Distance(A=Multiply(A=@n1) {constB=2}, B=ScalarParameter() {defaultValue=0.9214})
   Min=ScalarParameter() {defaultValue=0.0677, parameterName=min}
   Max=ScalarParameter() {defaultValue=0.0429, parameterName=max}
  ) {materialFunction=&SmoothStep}
  MaterialFunctionCall(Emissive=ParticleColor().RGB) {materialFunction=&MF_FX_EmissiveAdjustments}
 ]
}
